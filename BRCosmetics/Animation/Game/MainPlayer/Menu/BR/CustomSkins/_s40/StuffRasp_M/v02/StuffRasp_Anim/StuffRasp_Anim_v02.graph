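# Level Sequence that animates the StuffRasp_M v02 custom-skin menu performance on a spawned FortPlayerMannequin at 30 fps.
# A skeletal animation section plays under a Fortnite_3L_Face_CtrlRig track that layers on hand-keyed facial motion.
# Keys drive blinks, gaze, eyelids, squints, cheek raises and asymmetric brows, ending close to their starting pose.
# Jaw, mouth, tongue and aim controls stay neutral, while the upper nose wrinkle is held constant.
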LevelSequence /BRCosmetics/Animation/Game/MainPlayer/Menu/BR/CustomSkins/_s40/StuffRasp_M/v02/StuffRasp_Anim/StuffRasp_Anim_v02
bindings [FortPlayerMannequin {
  id=#1
  possessed_class=FortPlayerMannequin
  tracks [
   {type=MovieSceneSpawnTrack, sections [{type=MovieSceneSpawnSection, channels [None {type=MovieSceneScriptingBoolChannel, default=true}]}]}
   {
    type=MovieScene3DTransformTrack
    sections [{
      type=MovieScene3DTransformSection
      channels [
       Location.X {type=MovieSceneScriptingDoubleChannel, default=0}
       Location.Y {type=MovieSceneScriptingDoubleChannel, default=0}
       Location.Z {type=MovieSceneScriptingDoubleChannel, default=0}
       Rotation.X {type=MovieSceneScriptingDoubleChannel, default=0}
       Rotation.Y {type=MovieSceneScriptingDoubleChannel, default=0}
       Rotation.Z {type=MovieSceneScriptingDoubleChannel, default=0}
       Scale.X {type=MovieSceneScriptingDoubleChannel, default=1}
       Scale.Y {type=MovieSceneScriptingDoubleChannel, default=1}
       Scale.Z {type=MovieSceneScriptingDoubleChannel, default=1}
      ]
     }]
   }
   {type=MovieSceneSkeletalAnimationTrack, sections [{type=MovieSceneSkeletalAnimationSection, end_frame=269, start_frame=0}]}
   {
    type=MovieSceneControlRigParameterTrack
    display_name=Fortnite_3L_Face_CtrlRig
    sections [{
      type=MovieSceneControlRigParameterSection
      channels [
       CTRL_C_eyesAim.Location.X {type=MovieSceneScriptingFloatChannel, default=0, const [0,18]}
       CTRL_C_eyesAim.Location.Y {type=MovieSceneScriptingFloatChannel, default=0, const [0,18]}
       CTRL_C_eyesAim.Location.Z {type=MovieSceneScriptingFloatChannel, default=0, const [0,18]}
       CTRL_C_eyesAim.Rotation.X {type=MovieSceneScriptingFloatChannel, default=0, const [0,18]}
       CTRL_C_eyesAim.Rotation.Y {type=MovieSceneScriptingFloatChannel, default=0, const [0,18]}
       CTRL_C_eyesAim.Rotation.Z {type=MovieSceneScriptingFloatChannel, default=0, const [0,18]}
       CTRL_C_eyesAim.Scale.X {type=MovieSceneScriptingFloatChannel, default=1, const [1,18]}
       CTRL_C_eyesAim.Scale.Y {type=MovieSceneScriptingFloatChannel, default=1, const [1,18]}
       CTRL_C_eyesAim.Scale.Z {type=MovieSceneScriptingFloatChannel, default=1, const [1,18]}
       CTRL_L_eyeAim.Location.X {type=MovieSceneScriptingFloatChannel, default=0, const [0,18]}
       CTRL_L_eyeAim.Location.Y {type=MovieSceneScriptingFloatChannel, default=0, const [0,18]}
       CTRL_L_eyeAim.Location.Z {type=MovieSceneScriptingFloatChannel, default=0, const [0,18]}
       CTRL_L_eyeAim.Rotation.X {type=MovieSceneScriptingFloatChannel, default=0, const [0,18]}
       CTRL_L_eyeAim.Rotation.Y {type=MovieSceneScriptingFloatChannel, default=0, const [0,18]}
       CTRL_L_eyeAim.Rotation.Z {type=MovieSceneScriptingFloatChannel, default=0, const [0,18]}
       CTRL_L_eyeAim.Scale.X {type=MovieSceneScriptingFloatChannel, default=1, const [1,18]}
       CTRL_L_eyeAim.Scale.Y {type=MovieSceneScriptingFloatChannel, default=1, const [1,18]}
       CTRL_L_eyeAim.Scale.Z {type=MovieSceneScriptingFloatChannel, default=1, const [1,18]}
       CTRL_R_eyeAim.Location.X {type=MovieSceneScriptingFloatChannel, default=0, const [0,18]}
       CTRL_R_eyeAim.Location.Y {type=MovieSceneScriptingFloatChannel, default=0, const [0,18]}
       CTRL_R_eyeAim.Location.Z {type=MovieSceneScriptingFloatChannel, default=0, const [0,18]}
       CTRL_R_eyeAim.Rotation.X {type=MovieSceneScriptingFloatChannel, default=0, const [0,18]}
       CTRL_R_eyeAim.Rotation.Y {type=MovieSceneScriptingFloatChannel, default=0, const [0,18]}
       CTRL_R_eyeAim.Rotation.Z {type=MovieSceneScriptingFloatChannel, default=0, const [0,18]}
       CTRL_R_eyeAim.Scale.X {type=MovieSceneScriptingFloatChannel, default=1, const [1,18]}
       CTRL_R_eyeAim.Scale.Y {type=MovieSceneScriptingFloatChannel, default=1, const [1,18]}
       CTRL_R_eyeAim.Scale.Z {type=MovieSceneScriptingFloatChannel, default=1, const [1,18]}
       CTRL_convergenceSwitch {type=MovieSceneScriptingFloatChannel, default=0, const [0,18]}
       CTRL_Master.Location.X {type=MovieSceneScriptingFloatChannel, default=0, const [0,18]}
       CTRL_Master.Location.Y {type=MovieSceneScriptingFloatChannel, default=0, const [0,18]}
       CTRL_Master.Location.Z {type=MovieSceneScriptingFloatChannel, default=0, const [0,18]}
       CTRL_Master.Rotation.X {type=MovieSceneScriptingFloatChannel, default=0, const [0,18]}
       CTRL_Master.Rotation.Y {type=MovieSceneScriptingFloatChannel, default=0, const [0,18]}
       CTRL_Master.Rotation.Z {type=MovieSceneScriptingFloatChannel, default=0, const [0,18]}
       CTRL_Master.Scale.X {type=MovieSceneScriptingFloatChannel, default=1, const [1,18]}
       CTRL_Master.Scale.Y {type=MovieSceneScriptingFloatChannel, default=1, const [1,18]}
       CTRL_Master.Scale.Z {type=MovieSceneScriptingFloatChannel, default=1, const [1,18]}
       CTRL_C_jaw.X {type=MovieSceneScriptingFloatChannel, default=0, const [0,18]}
       CTRL_C_jaw.Y {type=MovieSceneScriptingFloatChannel, default=0, const [0,18]}
       CTRL_C_jaw_openExtreme {type=MovieSceneScriptingFloatChannel, default=0, const [0,18]}
       CTRL_C_jaw_fwdBack {type=MovieSceneScriptingFloatChannel, default=0, const [0,18]}
       CTRL_R_jaw_ChinRaiseD {type=MovieSceneScriptingFloatChannel, default=0, const [0,18]}
       CTRL_R_jaw_ChinRaiseU {type=MovieSceneScriptingFloatChannel, default=0, const [0,18]}
       CTRL_L_jaw_ChinRaiseU {type=MovieSceneScriptingFloatChannel, default=0, const [0,18]}
       CTRL_L_jaw_ChinRaiseD {type=MovieSceneScriptingFloatChannel, default=0, const [0,18]}
       CTRL_R_mouth_lipsBlow {type=MovieSceneScriptingFloatChannel, default=0, const [0,18]}
       CTRL_L_mouth_lipsTogetherU {type=MovieSceneScriptingFloatChannel, default=0, const [0,18]}
       CTRL_L_mouth_lipsTogetherD {type=MovieSceneScriptingFloatChannel, default=0, const [0,18]}
       CTRL_R_mouth_lipsTogetherU {type=MovieSceneScriptingFloatChannel, default=0, const [0,18]}
       CTRL_L_mouth_lipsBlow {type=MovieSceneScriptingFloatChannel, default=0, const [0,18]}
       CTRL_R_mouth_lipsTogetherD {type=MovieSceneScriptingFloatChannel, default=0, const [0,18]}
       CTRL_C_tongue_roll.X {type=MovieSceneScriptingFloatChannel, default=0, const [0,18]}
       CTRL_C_tongue_roll.Y {type=MovieSceneScriptingFloatChannel, default=0, const [0,18]}
       CTRL_C_tongue.X {type=MovieSceneScriptingFloatChannel, default=0, const [0,18]}
       CTRL_C_tongue.Y {type=MovieSceneScriptingFloatChannel, default=0, const [0,18]}
       CTRL_C_tongue_narrowWide {type=MovieSceneScriptingFloatChannel, default=0, const [0,18]}
       CTRL_C_tongue_inOut {type=MovieSceneScriptingFloatChannel, default=0, const [0,18]}
       CTRL_R_nose.X {type=MovieSceneScriptingFloatChannel, default=0, const [0,18]}
       CTRL_R_nose.Y {type=MovieSceneScriptingFloatChannel, default=0, const [0,18]}
       CTRL_L_nose.X {type=MovieSceneScriptingFloatChannel, default=0, const [0,18]}
       CTRL_L_nose.Y {type=MovieSceneScriptingFloatChannel, default=0, const [0,18]}
       CTRL_L_eye.X {type=MovieSceneScriptingFloatChannel, default=0, const [0,25]}
       CTRL_L_eye.Y {type=MovieSceneScriptingFloatChannel, default=0, const [0,25]}
       CTRL_L_eye_blink {
        type=MovieSceneScriptingFloatChannel
        default=0
        thinned_from=38
        keys [[0,0],[25,0],[27,0.0738],[30,1],[33,1],[37,0],[71,0],[74,1],[78,1],[87,0.0632],[97,0],[112,0],[116,1],[125,1],[135,0],[150,0],[157,0.0685],[160,1],[162,1],[169,0.0351],[173,0],[211,0],[218,0.0685],[219,0.1959],[220,0.4434],[221,1],[222,1],[229,0.158],[233,0],[294,0]]
       }
       CTRL_R_eye.X {type=MovieSceneScriptingFloatChannel, default=0, const [0,25]}
       CTRL_R_eye.Y {type=MovieSceneScriptingFloatChannel, default=0, const [0,25]}
       CTRL_R_eye_blink {
        type=MovieSceneScriptingFloatChannel
        default=0
        thinned_from=38
        keys [[0,0],[25,0],[27,0.0738],[30,1],[33,1],[37,0],[71,0],[74,1],[78,1],[87,0.0632],[97,0],[112,0],[116,1],[125,1],[135,0],[150,0],[157,0.0685],[160,1],[162,1],[169,0.0351],[173,0],[211,0],[218,0.0685],[219,0.1959],[220,0.4434],[221,1],[222,1],[229,0.158],[233,0],[294,0]]
       }
       CTRL_C_eye.X {
        type=MovieSceneScriptingFloatChannel
        default=0
        thinned_from=53
        keys [[0,0.0521],[9,0.0521],[10,0.0472],[11,0.0276],[23,0.0276],[26,0.1052],[27,0.2629],[29,0.3473],[45,0.3805],[60,0.3805],[62,0.3629],[63,0.247],[64,0.2326],[68,0.2326],[69,0.1716],[72,-0.2821],[75,-0.3556],[110,-0.3556],[112,-0.272],[114,-0.0333],[115,0.1565],[129,0.1474],[133,0.1474],[141,0.1446],[160,0.1446],[163,0.0276],[192,0.0276],[193,0.0623],[194,0.0742],[221,0.0742],[227,0.0963],[247,0.0963],[248,0.0908],[249,0.0521],[278,0.0521],[279,0.0472],[280,0.0276],[294,0.0276]]
       }
       CTRL_C_eye.Y {
        type=MovieSceneScriptingFloatChannel
        default=0
        thinned_from=52
        keys [[0,-0.0272],[9,-0.0272],[10,-0.0217],[11,0],[48,0],[49,-0.0392],[62,-0.0392],[63,-0.0718],[68,-0.0718],[69,-0.0721],[72,-0.0742],[75,-0.0664],[87,0],[110,0],[112,-0.0214],[114,-0.1354],[115,-0.2066],[129,-0.4161],[133,-0.1176],[141,-0.1176],[156,0.1569],[160,0.153],[162,-0.0859],[165,-0.0859],[172,0.0282],[184,0],[207,0],[208,0.0502],[209,0.0621],[219,0.0621],[220,-0.0918],[221,-0.1856],[227,-0.0797],[228,-0.0459],[229,0.0621],[247,0.0621],[248,0.0509],[249,-0.0272],[278,-0.0272],[279,-0.0217],[280,0],[294,0.0001]]
       }
       CTRL_C_eyesConvergence {type=MovieSceneScriptingFloatChannel, default=0, const [0,25]}
       CTRL_L_eye_cheekRaise {
        type=MovieSceneScriptingFloatChannel
        default=0
        thinned_from=27
        keys [[0,0.6081],[23,0.6111],[30,0.6111],[45,0],[69,0],[72,0.0741],[87,1],[125,1],[135,0.4269],[141,0.4269],[150,1],[160,0.9761],[162,0.9618],[169,0.4745],[173,0.4568],[180,0.6249],[184,0.6739],[211,0.6809],[221,0.6255],[222,0.4282],[227,0.5772],[243,0.6009],[269,0.6081],[294,0.6112]]
       }
       CTRL_L_eye_squintInner {
        type=MovieSceneScriptingFloatChannel
        default=0
        thinned_from=26
        keys [[0,0.0846],[21,0.1322],[30,0.1322],[37,1],[45,0.43],[141,0.43],[150,0.2985],[160,0.2802],[162,0.2692],[169,0.0094],[173,0],[180,0.5709],[184,0.6823],[211,0.6982],[227,0.034],[243,0.034],[269,0.0846],[294,0.1322]]
       }
       CTRL_L_nose_wrinkleUpper {type=MovieSceneScriptingFloatChannel, default=0.9008, const [1,25]}
       CTRL_R_nose_wrinkleUpper {type=MovieSceneScriptingFloatChannel, default=1, const [1,25]}
       CTRL_R_eye_squintInner {
        type=MovieSceneScriptingFloatChannel
        default=0
        thinned_from=26
        keys [[0,0.0846],[21,0.1322],[30,0.1322],[37,1],[45,0],[69,0],[72,0.0514],[87,0.6941],[125,0.6941],[135,0.3342],[141,0.3342],[150,0.232],[160,0.2178],[162,0.2092],[169,0.0073],[173,0],[180,0.5451],[184,0.704],[211,0.7268],[227,0.034],[243,0.034],[269,0.0846],[294,0.1322]]
       }
       CTRL_R_eye_cheekRaise {
        type=MovieSceneScriptingFloatChannel
        default=0
        thinned_from=27
        keys [[0,0.7469],[30,0.75],[69,0.75],[72,0.6944],[87,0],[125,0],[135,0.7753],[141,0.7753],[150,1],[160,0.9846],[162,0.9755],[169,0.609],[173,0.5957],[180,0.7638],[184,0.8128],[211,0.8198],[221,0.7644],[222,0.5671],[227,0.7161],[243,0.7398],[269,0.7469],[294,0.75]]
       }
       CTRL_C_eyelidsReduce {
        type=MovieSceneScriptingFloatChannel
        default=1
        thinned_from=28
        keys [[0,0.8393],[23,0.8393],[30,1],[135,1],[141,0.9858],[150,0.927],[160,0.9216],[162,0.9184],[169,0.8611],[175,0.8393],[294,0.8393]]
       }
       CTRL_R_brow_lateral {
        type=MovieSceneScriptingFloatChannel
        default=0
        thinned_from=28
        keys [[0,0.9999],[72,1],[75,0.9618],[87,0.6329],[125,0.6329],[135,0.7902],[141,0.8472],[150,1],[162,1],[169,0.8218],[175,0.7538],[184,0.9326],[211,0.9582],[218,0.9634],[223,1],[230,1.0197],[243,1.0263],[269,0.9999],[294,0.9999]]
       }
       CTRL_L_brow_lateral {
        type=MovieSceneScriptingFloatChannel
        default=0
        thinned_from=28
        keys [[0,1],[30,1],[37,0.9345],[45,0.7381],[72,0.7366],[75,0.764],[87,1],[125,1],[135,0.7902],[141,0.8087],[150,1],[162,1],[169,0.6813],[175,0.5597],[184,0.876],[211,0.9212],[218,0.9311],[223,1],[230,1.0313],[243,1.0417],[269,1],[294,1]]
       }
       CTRL_R_brow_down {
        type=MovieSceneScriptingFloatChannel
        default=1
        thinned_from=30
        keys [[0,0.7435],[23,0.7485],[26,0.7634],[30,0.8675],[37,0.844],[45,0.7734],[69,0.7734],[72,0.8743],[75,0.9505],[87,0.4239],[125,0.4239],[135,0.7734],[141,0.7712],[150,0.8526],[160,0.9103],[162,0.9058],[169,0.5956],[175,0.4772],[184,0.6485],[211,0.7876],[218,0.8097],[223,0.9642],[230,0.7947],[243,0.7382],[294,0.7485]]
       }
       CTRL_L_brow_down {
        type=MovieSceneScriptingFloatChannel
        default=1
        thinned_from=30
        keys [[0,0.7435],[23,0.7485],[26,0.7634],[30,0.8675],[37,0.7192],[45,0.2742],[69,0.2714],[72,0.3723],[75,0.4485],[87,0.7493],[125,0.7493],[135,0.4707],[141,0.4953],[150,0.6824],[160,0.7552],[162,0.7567],[169,0.5545],[175,0.4772],[184,0.6485],[211,0.7876],[218,0.8097],[223,0.9642],[230,0.7947],[243,0.7382],[294,0.7485]]
       }
       CTRL_L_brow_raiseOut {
        type=MovieSceneScriptingFloatChannel
        default=0.6296
        thinned_from=28
        keys [[0,0.5985],[23,0.6011],[69,0.6011],[72,0.3743],[75,0.2713],[87,0],[125,0],[135,0.3058],[141,0.3861],[150,0.431],[160,0.452],[162,0.4645],[169,0.628],[175,0.6905],[184,0.6282],[218,0.6172],[223,0.6027],[230,0.5974],[243,0.5956],[294,0.6012]]
       }
       CTRL_L_brow_raiseIn {
        type=MovieSceneScriptingFloatChannel
        default=0
        thinned_from=28
        keys [[0,0],[30,0],[37,0.0306],[45,0.1225],[69,0.1232],[72,0.0767],[75,0.0556],[87,0],[125,0],[135,0.1225],[141,0.1117],[150,0],[162,0],[169,0.0647],[175,0.0894],[184,0.0271],[211,0.0182],[218,0.0162],[223,0.0016],[230,-0.0037],[243,-0.0055],[269,0],[294,0.0001]]
       }
       CTRL_R_brow_raiseIn {
        type=MovieSceneScriptingFloatChannel
        default=1
        thinned_from=28
        keys [[0,0],[30,0],[37,0.0248],[45,0.0993],[69,0.0998],[72,0.0622],[75,0.0451],[87,0],[125,0],[135,0.0993],[141,0.0905],[150,0],[162,0],[169,0.0647],[175,0.0894],[184,0.0271],[211,0.0182],[218,0.0162],[223,0.0016],[230,-0.0037],[243,-0.0055],[269,0],[294,0.0001]]
       }
       CTRL_R_brow_raiseOut {
        type=MovieSceneScriptingFloatChannel
        default=0
        thinned_from=28
        keys [[0,0.2028],[21,0.2054],[69,0.2054],[72,0.1359],[75,0.1042],[87,0.021],[125,0.021],[135,0.2054],[162,0.2054],[169,0.2701],[175,0.2948],[184,0.2325],[218,0.2216],[223,0.207],[230,0.2017],[243,0.1999],[294,0.2055]]
       }
       CTRL_lookAtSwitch {type=MovieSceneScriptingFloatChannel, default=0, const [0,18]}
       CTRL_C_teethD.X {type=MovieSceneScriptingFloatChannel, default=0, const [0,18]}
       CTRL_C_teethD.Y {type=MovieSceneScriptingFloatChannel, default=0, const [0,18]}
       CTRL_C_teethU.X {type=MovieSceneScriptingFloatChannel, default=0, const [0,18]}
       CTRL_C_teethU.Y {type=MovieSceneScriptingFloatChannel, default=0, const [0,18]}
       CTRL_C_teeth_fwdBackU {type=MovieSceneScriptingFloatChannel, default=0, const [0,18]}
       CTRL_L_mouth_corner.X {type=MovieSceneScriptingFloatChannel, default=0, const [0,18]}
       CTRL_L_mouth_corner.Y {type=MovieSceneScriptingFloatChannel, default=0, const [0,18]}
       CTRL_R_mouth_corner.X {type=MovieSceneScriptingFloatChannel, default=0, const [0,19]}
       CTRL_R_mouth_corner.Y {type=MovieSceneScriptingFloatChannel, default=0, const [0,18]}
       CTRL_C_teeth_fwdBackD {type=MovieSceneScriptingFloatChannel, default=0, const [0,18]}
       CTRL_R_mouth_cornerSharpnessD {type=MovieSceneScriptingFloatChannel, default=0, const [0,18]}
       CTRL_R_mouth_cornerSharpnessU {type=MovieSceneScriptingFloatChannel, default=0, const [0,18]}
       CTRL_L_eye_eyelidU.X {
        type=MovieSceneScriptingFloatChannel
        default=0
        thinned_from=30
        keys [[0,0],[33,0],[37,-0.2406],[116,-0.2406],[125,-0.5191],[135,-0.2193],[141,-0.2193],[150,-0.1602],[157,-0.1523],[160,0],[294,0]]
       }
       CTRL_L_eye_eyelidU.Y {
        type=MovieSceneScriptingFloatChannel
        default=0
        thinned_from=31
        keys [[0,-0.0457],[27,-0.0457],[30,-0.0714],[33,-0.2058],[37,-0.1688],[69,-0.1688],[72,-0.1774],[74,-0.2171],[77,-0.409],[87,-0.1021],[112,-0.1021],[116,-0.4688],[125,-0.6472],[135,-0.1229],[150,-0.1229],[157,-0.1616],[160,0.0667],[162,-0.0098],[169,-0.0003],[173,0],[180,-0.0457],[294,-0.0457]]
       }
       CTRL_L_mouth_cornerSharpnessD {type=MovieSceneScriptingFloatChannel, default=0, const [0,18]}
       CTRL_L_eye_eyelidD.X {type=MovieSceneScriptingFloatChannel, default=0, thinned_from=30, keys [[0,0],[294,0]]}
       CTRL_L_eye_eyelidD.Y {
        type=MovieSceneScriptingFloatChannel
        default=0
        thinned_from=31
        keys [[0,-0.0001],[27,0],[30,-0.0257],[33,-0.2058],[37,0],[69,0],[72,-0.0086],[74,-0.0484],[77,-0.2403],[87,0.0667],[112,0.0667],[116,-0.3],[125,-0.4784],[135,0],[141,0],[150,0.2351],[157,0.2667],[160,0.0667],[162,-0.0098],[169,-0.0003],[173,0],[294,0]]
       }
       CTRL_L_mouth_cornerSharpnessU {type=MovieSceneScriptingFloatChannel, default=0, const [0,18]}
       CTRL_R_mouth_lipsTowardsTeethD {type=MovieSceneScriptingFloatChannel, default=0, const [0,18]}
       CTRL_L_mouth_lipsTowardsTeethD {type=MovieSceneScriptingFloatChannel, default=0, const [0,18]}
       CTRL_R_eye_eyelidU.X {type=MovieSceneScriptingFloatChannel, default=0, thinned_from=30, keys [[0,0],[141,0],[150,-0.0588],[157,-0.0667],[160,0],[294,0]]}
       CTRL_R_eye_eyelidU.Y {
        type=MovieSceneScriptingFloatChannel
        default=0
        thinned_from=31
        keys [[0,-0.0457],[27,-0.0457],[30,-0.0714],[33,-0.2058],[37,-0.3375],[69,-0.3375],[72,-0.3461],[74,-0.3859],[77,-0.5778],[87,-0.2708],[112,-0.2708],[116,-0.6375],[125,-0.8159],[135,-0.2513],[150,-0.2513],[157,-0.29],[160,0.0667],[162,-0.0098],[169,-0.0003],[173,0],[180,-0.0457],[294,-0.0457]]
       }
       CTRL_L_mouth_lipsTowardsTeethU {type=MovieSceneScriptingFloatChannel, default=0, const [0,18]}
       CTRL_R_eye_eyelidD.X {type=MovieSceneScriptingFloatChannel, default=0, thinned_from=30, keys [[0,0],[294,0]]}
       CTRL_R_eye_eyelidD.Y {
        type=MovieSceneScriptingFloatChannel
        default=0
        thinned_from=31
        keys [[0,-0.0001],[27,0],[30,-0.0257],[33,-0.2058],[37,0],[69,0],[72,-0.0086],[74,-0.0484],[77,-0.2403],[87,0.0667],[112,0.0667],[116,-0.3],[125,-0.4784],[135,0],[141,0],[150,0.2351],[157,0.2667],[160,0.0667],[162,-0.0098],[169,-0.0003],[173,0],[294,0]]
       }
       CTRL_R_mouth_lipsTowardsTeethU {type=MovieSceneScriptingFloatChannel, default=0, const [0,18]}
       CTRL_faceGUIfollowHead {type=MovieSceneScriptingFloatChannel, default=0, const [0,18]}
       CTRL_eyesAimFollowHead {type=MovieSceneScriptingFloatChannel, default=0, const [0,18]}
       CTRL_C_nose.X {type=MovieSceneScriptingFloatChannel, default=0, const [0,18]}
       CTRL_C_nose.Y {type=MovieSceneScriptingFloatChannel, default=0, const [0,18]}
       CTRL_C_teeth_rotateU {type=MovieSceneScriptingFloatChannel, default=0, const [0,18]}
       CTRL_C_teeth_rotateD {type=MovieSceneScriptingFloatChannel, default=0, const [0,18]}
       CTRL_L_brow_inner {
        type=MovieSceneScriptingFloatChannel
        default=-1
        thinned_from=28
        keys [[0,-0.0223],[30,-0.0222],[37,-0.19],[45,-0.6932],[69,-0.697],[72,-0.5551],[75,-0.7855],[87,-0.2889],[125,-0.2889],[135,-0.6932],[141,-0.6338],[150,-0.3885],[160,-0.366],[162,-0.3526],[169,-0.2928],[175,-0.27],[184,-0.1074],[211,-0.0842],[218,-0.0787],[223,-0.0407],[230,-0.0268],[243,-0.0222],[294,-0.0222]]
       }
       CTRL_R_brow_inner {
        type=MovieSceneScriptingFloatChannel
        default=0
        thinned_from=28
        keys [[0,-0.4222],[30,-0.4222],[37,-0.4611],[45,-0.5778],[69,-0.5787],[72,-0.4707],[75,-0.7011],[87,-0.6889],[125,-0.6889],[135,-0.5556],[141,-0.7415],[150,-0.6439],[160,-0.6303],[162,-0.6221],[169,-0.586],[175,-0.5722],[184,-0.4738],[218,-0.4564],[223,-0.4334],[230,-0.425],[243,-0.4222],[294,-0.4222]]
       }
       CTRL_L_brow_outer {
        type=MovieSceneScriptingFloatChannel
        default=-1
        thinned_from=28
        keys [[0,-1],[30,-1],[37,-0.8664],[45,-0.4654],[69,-0.4624],[72,-0.4976],[75,-0.7437],[87,-0.5556],[125,-0.5556],[135,-0.1099],[141,-0.1887],[150,-0.5141],[160,-0.5439],[162,-0.5618],[169,-0.8789],[175,-1],[294,-1]]
       }
       CTRL_R_brow_outer {
        type=MovieSceneScriptingFloatChannel
        default=0
        thinned_from=28
        keys [[0,-0.9556],[69,-0.9556],[72,-0.7544],[75,-0.8705],[87,-0.4222],[125,-0.4222],[135,-0.9556],[294,-0.9556]]
       }
       CTRL_C_mouth.X {type=MovieSceneScriptingFloatChannel, default=0, const [0,18]}
       CTRL_C_mouth.Y {type=MovieSceneScriptingFloatChannel, default=0, const [0,18]}
       CTRL_L_mouth_upperLipRaise {type=MovieSceneScriptingFloatChannel, default=0, const [0,18]}
       CTRL_L_mouth_cornerPull {type=MovieSceneScriptingFloatChannel, default=0, const [0,18]}
       CTRL_L_mouth_dimple {type=MovieSceneScriptingFloatChannel, default=0, const [0,18]}
       CTRL_L_mouth_cornerDepress {type=MovieSceneScriptingFloatChannel, default=0, const [0,18]}
       CTRL_L_mouth_stretch {type=MovieSceneScriptingFloatChannel, default=0, const [0,18]}
       CTRL_L_mouth_lowerLipDepress {type=MovieSceneScriptingFloatChannel, default=0, const [0,18]}
       CTRL_R_mouth_lowerLipDepress {type=MovieSceneScriptingFloatChannel, default=0, const [0,18]}
       CTRL_R_mouth_stretch {type=MovieSceneScriptingFloatChannel, default=0, const [0,18]}
       CTRL_L_mouth_towardsD {type=MovieSceneScriptingFloatChannel, default=0, const [0,18]}
       CTRL_R_mouth_funnelU {type=MovieSceneScriptingFloatChannel, default=0, const [0,18]}
       CTRL_L_mouth_towardsU {type=MovieSceneScriptingFloatChannel, default=0, const [0,18]}
       CTRL_R_mouth_lipBiteD {type=MovieSceneScriptingFloatChannel, default=0, const [0,18]}
       CTRL_R_mouth_funnelD {type=MovieSceneScriptingFloatChannel, default=0, const [0,18]}
       CTRL_L_mouth_pressU {type=MovieSceneScriptingFloatChannel, default=0, const [0,18]}
       CTRL_L_mouth_funnelD {type=MovieSceneScriptingFloatChannel, default=0, const [0,18]}
       CTRL_L_mouth_suckBlow {type=MovieSceneScriptingFloatChannel, default=0, const [0,18]}
       CTRL_L_mouth_pressD {type=MovieSceneScriptingFloatChannel, default=0, const [0,18]}
       CTRL_L_mouth_funnelU {type=MovieSceneScriptingFloatChannel, default=0, const [0,18]}
       CTRL_R_mouth_towardsU {type=MovieSceneScriptingFloatChannel, default=0, const [0,18]}
       CTRL_R_mouth_purseD {type=MovieSceneScriptingFloatChannel, default=0, const [0,18]}
       CTRL_L_mouth_lipBiteU {type=MovieSceneScriptingFloatChannel, default=0, const [0,18]}
       CTRL_R_mouth_pressD {type=MovieSceneScriptingFloatChannel, default=0, const [0,18]}
       CTRL_R_mouth_pressU {type=MovieSceneScriptingFloatChannel, default=0, const [0,18]}
       CTRL_L_mouth_purseD {type=MovieSceneScriptingFloatChannel, default=0, const [0,18]}
       CTRL_R_mouth_lipBiteU {type=MovieSceneScriptingFloatChannel, default=0, const [0,18]}
       CTRL_R_mouth_suckBlow {type=MovieSceneScriptingFloatChannel, default=0, const [0,18]}
       CTRL_L_mouth_purseU {type=MovieSceneScriptingFloatChannel, default=0, const [0,18]}
       CTRL_R_mouth_towardsD {type=MovieSceneScriptingFloatChannel, default=0, const [0,18]}
       CTRL_L_mouth_lipBiteD {type=MovieSceneScriptingFloatChannel, default=0, const [0,18]}
       CTRL_R_mouth_purseU {type=MovieSceneScriptingFloatChannel, default=0, const [0,18]}
       CTRL_R_mouth_cornerDepress {type=MovieSceneScriptingFloatChannel, default=0, const [0,18]}
       CTRL_R_mouth_upperLipRaise {type=MovieSceneScriptingFloatChannel, default=0, const [0,18]}
       CTRL_R_mouth_dimple {type=MovieSceneScriptingFloatChannel, default=0, const [0,18]}
       CTRL_R_mouth_cornerPull {type=MovieSceneScriptingFloatChannel, default=0, const [0,18]}
      ]
     }]
   }
  ]
 }]
display_rate {denominator=1, numerator=30}
tick_resolution {denominator=1, numerator=24000}
tracks []
